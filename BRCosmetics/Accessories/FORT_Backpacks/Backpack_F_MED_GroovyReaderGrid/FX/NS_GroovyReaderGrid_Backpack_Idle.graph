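# Idle effect for the GroovyReaderGrid back bling: mesh syrup drips that form on four drip bones and stop spawning while skydiving.
# Each drip follows its skinned bone on a lengthening pendulum, stretching and orienting along it, with an age threshold switching off the constraint for release.
# Gravity, drag and late-life opposing owner velocity drive motion.
# Color and opacity vary by time of day and front-end state; a debug emitter is disabled.
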NiagaraSystem /BRCosmetics/Accessories/FORT_Backpacks/Backpack_F_MED_GroovyReaderGrid/FX/NS_GroovyReaderGrid_Backpack_Idle
dependencies {
 usedDataInterfaces [&NiagaraDataInterfaceArrayFloat, &NiagaraDataInterfaceCurve]
 usedDynamicInputs []
 usedModules [
  &SystemState
  &self:SystemSpawnScript.SystemScriptSource.SystemScriptGraph.NiagaraNodeAssignment_2.SetVariables_#1
  &Timeline
 ]
}
emitters {
 FN_Debug {
  bEnabled=false
  emitterUpdateScript [
   &EmitterState()
   &SpawnBurst_Instantaneous(Spawn Count=1, Use Spawn Probability=false)
  ]
  particleSpawnScript [/Niagara/Modules/Spawn/Initialization/InitializeParticle(
    Lifetime=5
    Module.Write Color=true
    Module.Write Lifetime=true
    Module.Write Mass=true
    Module.Write Material Random=true
    Module.Write Position=false
    Module.Write Scale=false
    Module.Write SpriteRotation=false
    Module.Write SpriteSize=true
    Position=&SimulationPosition()
    Sprite Size=<100 100>
   )]
  particleUpdateScript [
   &ParticleState()
   &FN_Timer()
   /SetVariables(Particles.Debug:NiagaraFloat=$System.FN_Timer.Value) {isSetParameters=true}
   &DynamicMaterialParameters(
    Index 0 Param 1=$Particles.Debug
    Module.Param0WriteEnabled=true
    Module.Param1WriteEnabled=true
    Module.Param2WriteEnabled=true
    Module.Param3WriteEnabled=true
   )
   &CameraOffset(Camera Offset Amount=111)
  ]
  rendererClasses [&NiagaraSpriteRendererProperties]
  renderers [{rendererClass=&NiagaraSpriteRendererProperties}]
 }
 SyrupDrips {
  emitterUpdateScript [
   &EmitterState(Life Cycle Mode=::NewEnumerator1, Loop Behavior=::NewEnumerator1, Loop Duration Mode=::NewEnumerator1, UseLoopDelay=false)
   &SpawnRate(
    SpawnRate=&MakeCustomFloatFromBool(Bool=$User.IsSkydiving, True Float=0, False Float=0.5)
    Use Spawn Probability=false
   )
  ]
  particleSpawnScript [
   &PartitionParticles(Number Of Partitions=4)
   /SetVariables(Particles.BackpackSelectedBone:NiagaraInt32=$Particles.PartitionParticles.Partition) {isSetParameters=true}
   /Niagara/Modules/Spawn/Initialization/V2/InitializeParticle(Lifetime Max=4.3, Lifetime Min=3.6, Lifetime Mode=::NewEnumerator1, UsePositionOffset=false, Write Material Random=true)
   &SkeletalMeshLocation(
    Bone / Socket Index=$Particles.BackpackSelectedBone
    Bone Sampling Mode=::NewEnumerator4
    MaskSpawnGroup=false
    Skeletal Mesh=DataInterface() {propertyValues {
      DefaultMesh=null
      ExcludeBoneName=root
      OptionalComponentClass=null
      PreviewMesh=/BRCosmetics/Accessories/FORT_Backpacks/Backpack_F_MED_GroovyReaderGrid/Meshes/F_MED_GroovyReaderGrid_Pack
      SkinningMode=SkinOnTheFly
      SoftSourceActor=null
      SourceMode=Default
      UvSetIndex=0
      WholeMeshLOD=-1
      bExcludeBone=false
      bReadDeformedGeometry=true
      bRequireCurrentFrameData=true
      ComponentTags []
      FilteredBones [dyn_drip_02_l, dyn_drip_01_mid_r, dyn_drip_02_r, dyn_drip_mid_l]
      FilteredSockets []
      MeshUserParameter {parameter {name=None, typeDefHandle {registeredTypeIndex=1}, varData []}}
      SamplingRegions []
     }}
   )
   /SetVariables(Particles.Scale=&Multiply_VectorByFloat(
     Vector=&VectorFromFloat(Value=0.12)
     Float=&SelectFloatFromArray(Direct Array Index=$Particles.BackpackSelectedBone, Float Selection Array=DataInterface() {propertyValues {FloatData [0.7,0.85,0.65,0.62]}})
    )) {isSetParameters=true}
   &CameraOffset(Camera Offset Amount=2)
   &InitialMeshOrientation(Mesh Orientation Mode=::NewEnumerator2, Orientation Vector=<-1 0 0>, Use Rotation Vector=false) {enabled=false}
   &DynamicMaterialParameters(
    Index 0 Param 3=&FN_FrontEndFloatModifier(Gameplay=0, Front End=1)
    Module.Index0Param0Write=true
    Module.Index0Param1Write=true
    Module.Index0Param2Write=true
    Module.Index0Param3Write=false
   )
  ]
  particleUpdateScript [
   /SetVariables(Particles.Connection:NiagaraBool=&SetBoolByFloatComparison(A=$Particles.NormalizedAge, B=0.85)) {isSetParameters=true}
   &ScaleColor(
    Scale Alpha=&FloatFromCurve(FloatCurve=DataInterface() {propertyValues {Curve {keys [{interpMode=RCIM_Cubic, tangentMode=RCTM_User, time=0.2, value=0}, {interpMode=RCIM_Cubic, tangentMode=RCTM_User, time=0.45, value=1}]}}})
    Scale RGB=&FN_FrontEndVector3Modifier(
     Gameplay=&MakeVectorFromLinearColorRGB(Color=&FNTimeOfDayColor(
       Night Color=<rgba 2.8 2.128 0.28 1>
       Morning Color=<rgba 0.9 0.6345 0.369 1>
       Day Color=<rgba 0.65 0.4647 0.065 1>
       Evening Color=<rgba 1.1 0.8784 0.077 1>
      ))
     Front End=&MakeVectorFromLinearColorRGB(Color=<rgba 0.26 0.2436 0.1196 1>)
    )
    ScaleA=true
    ScaleRGB=true
   ) {enabled=false}
   &ScaleColor(
    Scale Alpha=&FloatFromCurve(FloatCurve=DataInterface() {propertyValues {Curve {keys [{interpMode=RCIM_Cubic, tangentMode=RCTM_User, time=0.2, value=0}, {interpMode=RCIM_Cubic, tangentMode=RCTM_User, time=0.45, value=1}]}}})
    Scale RGB=&FN_FrontEndVector3Modifier(
     Gameplay=&MakeVectorFromLinearColorRGB(Color=&FNTimeOfDayColor(
       Night Color=<rgba 1 0.439 0.01 1>
       Morning Color=<rgba 0.2 0.08 0.02 1>
       Day Color=<rgba 0.14 0.0711 0.0252 1>
       Evening Color=<rgba 0.27 0.0999 0.027 1>
      ))
     Front End=&MakeVectorFromLinearColorRGB(Color=<rgba 0.26 0.2436 0.1196 1>)
    )
    ScaleA=true
    ScaleRGB=true
   ) {enabled=false, module=ScaleColor001}
   /SetVariables(Particles.Color=&MakeLinearColorFromVectorAndFloat(
     "Vector (RGB)"=&FN_FrontEndVector3Modifier(
      Gameplay=&MakeVectorFromLinearColorRGB(Color=&FNTimeOfDayColor(
        Night Color=<rgba 1 0.3467 0.02 1>
        Morning Color=<rgba 0.16 0.0695 0.0208 1>
        Day Color=<rgba 0.32 0.14 0.032 1>
        Evening Color=<rgba 1 0.37 0.1 1>
       ))
      Front End=&MakeVectorFromLinearColorRGB(Color=<rgba 0.15 0.0668 0.015 1>)
     )
     "Float (Alpha)"=&FloatFromCurve(
      FloatCurve=DataInterface() {propertyValues {Curve {keys [{interpMode=RCIM_Cubic, tangentMode=RCTM_User, time=0.2, value=0}, {interpMode=RCIM_Cubic, tangentMode=RCTM_User, time=0.45, value=1}]}}}
      Scale Curve=&FN_FrontEndFloatModifier(
       Gameplay=&FNTimeOfDayScalar(Night Scale=0.17, Morning Scale=0.35, Day Scale=0.24, Evening Scale=0.25)
       Front End=0.27
      )
     )
    )) {isSetParameters=true}
   &ParticleState()
   &ScaleMeshSize(Scale Factor=&VectorFromFloat(Value=&FloatFromCurve(FloatCurve=DataInterface() {propertyValues {Curve {keys [
          {arriveTangent=0.6088, interpMode=RCIM_Cubic, leaveTangent=0.6088, tangentMode=RCTM_User, time=0.15, value=0.4}
          {interpMode=RCIM_Cubic, tangentMode=RCTM_User, time=0.7, value=1}
          {arriveTangent=-0.0171, interpMode=RCIM_Cubic, leaveTangent=-0.0171, tangentMode=RCTM_User, time=0.95, value=1}
          {arriveTangent=-22.3504, interpMode=RCIM_Cubic, leaveTangent=-22.3504, tangentMode=RCTM_User, time=1, value=0}
         ]}}})))
   &ScaleMeshSize(Scale Factor=&MakeVector(
     X=1
     Y=1
     Z=&FloatFromCurve(
      FloatCurve=DataInterface() {propertyValues {Curve {keys [
          {arriveTangent=4.2114, interpMode=RCIM_Cubic, leaveTangent=4.2114, tangentMode=RCTM_User, time=0.2, value=1}
          {arriveTangent=0.5863, interpMode=RCIM_Cubic, leaveTangent=0.5863, tangentMode=RCTM_User, time=1, value=2.5}
         ]}}}
      Scale Curve=&SelectFloatFromArray(Direct Array Index=$Particles.BackpackSelectedBone, Float Selection Array=DataInterface() {propertyValues {FloatData [1,1,1.5,0.8]}})
     )
    )) {module=ScaleMeshSize001}
   &SampleSkeletalMesh(
    AdditionalYawPitchRoll=false
    Bone / Socket Index:NiagaraInt32=$Particles.BackpackSelectedBone
    Bone Sampling Mode:ENiagaraBoneSamplingMode=::NewEnumerator4
    Skeletal Mesh=DataInterface() {propertyValues {
      DefaultMesh=null
      ExcludeBoneName=root
      OptionalComponentClass=null
      PreviewMesh=&F_MED_GroovyReaderGrid_Pack
      SkinningMode=SkinOnTheFly
      SoftSourceActor=null
      SourceMode=Default
      UvSetIndex=0
      WholeMeshLOD=-1
      bExcludeBone=false
      bReadDeformedGeometry=true
      bRequireCurrentFrameData=true
      ComponentTags []
      FilteredBones [dyn_drip_02_l, dyn_drip_01_mid_r, dyn_drip_02_r, dyn_drip_mid_l]
      FilteredSockets []
      MeshUserParameter {parameter {name=None, typeDefHandle {registeredTypeIndex=1}, varData []}}
      SamplingRegions []
     }}
   )
   /SetVariables(Particles.MeshOrientation=&FindQuatBetweenTwoVectors(
     From Vector=<0 0 -1>
     To Vector=&MakeCustomVectorFromBool(
      Bool=$Particles.Connection
      True Vector=&NormalizeVector(
       Vector To Normalize=&SubtractVectorFromPosition(Position=$Particles.Position, Vector=$Particles.SampleSkeletalMesh.SampledPosition)
       Safe Normalize=true
       Safe Fallback Threshold=0
       Safe Fallback Vector=<0 0 0.001>
      )
      False Vector=&NormalizeVector(
       Vector To Normalize=&Lerp_Vector(
        A=&SubtractVectorFromPosition(Position=$Particles.Position, Vector=$Particles.SampleSkeletalMesh.SampledPosition)
        B=<0 0 -1>
        Alpha=&FloatFromCurve(FloatCurve=DataInterface() {propertyValues {Curve {keys [<at 0.7 0>,<at 0.8 1>]}}})
       )
       Safe Normalize=true
       Safe Fallback Threshold=0
       Safe Fallback Vector=<0 0 0.001>
      )
     )
    )) {isSetParameters=true}
   &GravityForce()
   &Drag(Drag=2, Use Linear Drag=true, Use Rotational Drag=true)
   &PendulumSetup(
    Module.Enable Spring Driver=false
    Pendulum Length:NiagaraFloat=&FloatFromCurve(
     FloatCurve=DataInterface() {propertyValues {Curve {keys [<at 0 0 RCIM_Cubic>, {arriveTangent=0.0714, interpMode=RCIM_Cubic, leaveTangent=0.0714, tangentMode=RCTM_User, time=0.7, value=1}]}}}
     Scale Curve=&SelectFloatFromArray(Direct Array Index=$Particles.BackpackSelectedBone, Float Selection Array=DataInterface() {propertyValues {FloatData [2,2,2,1.5]}})
    )
    Pendulum Pivot:Vector3f=$Particles.SampleSkeletalMesh.SampledPosition
    Rigid Pendulum=false
    "Spring Driven Constraint (Tightness)"=2
   )
   &AccelerationForce(Acceleration=&Multiply_VectorByFloat(
     Vector=&Multiply_Vector(A=$Engine.Owner.Velocity, B=<-1 -1 0>)
     Float=&FloatFromCurve(FloatCurve=DataInterface() {propertyValues {Curve {keys [{interpMode=RCIM_Cubic, tangentMode=RCTM_User, time=0.6, value=0}, {interpMode=RCIM_Cubic, tangentMode=RCTM_User, time=0.9, value=1}]}}})
    ))
   &SolveForcesAndVelocity(Clamp Velocity=false, Limit Acceleration=false)
   /SetVariables(Particles.PositionTemp:NiagaraPosition=$Particles.Position) {isSetParameters=true}
   &ScaleMeshSizeBySpeed(Max Scale Factor=<1 1 0.3>, Min Scale Factor:Vector3f=<0.6 0.6 1>, Speed Threshold=300)
   &PendulumConstraint()
   /SetVariables(Particles.Position=&MakeCustomVectorFromBool(
     Bool=$Particles.Connection
     True Vector=Linked() {linkedVariable {name=Particles.PositionTemp, type=NiagaraPosition}}
     False Vector=Linked() {linkedVariable {name=Particles.Position, type=NiagaraPosition}}
    )) {isSetParameters=true}
   &CalculateAccurateVelocity()
   /SetVariables(Particles.PositionDrips:NiagaraPosition=&AddVectorToPosition(
     Position=$Particles.Position
     Vector=&Multiply_VectorByFloat(Vector=&RotateVectorByQuaternion(Vector To Rotate=<0 0 1>, Quaternion=$Particles.MeshOrientation), Float=1.6)
    )) {isSetParameters=true}
  ]
  rendererClasses [&NiagaraMeshRendererProperties]
  renderers [{rendererClass=&NiagaraMeshRendererProperties, rendererIndex=0}]
 }
}
summary {
 systemName=NS_GroovyReaderGrid_Backpack_Idle
 emitters [
  {bEnabled=false, emitterName=FN_Debug, rendererClasses [&NiagaraSpriteRendererProperties]}
  {emitterName=SyrupDrips, rendererClasses [&NiagaraMeshRendererProperties]}
 ]
 userVariables [
  User.bIsFrontEnd {type=NiagaraBool}
  User.bIsFrontEndPreview {type=NiagaraBool, defaultValue=true}
  User.Empowered {type=NiagaraBool}
  User.IsSkydiving {type=NiagaraBool}
  User.SkeletalMesh {
   type=NiagaraDataInterfaceSkeletalMesh
   defaultValue {
    struct=NiagaraExt_VariableValue_DataInterface
    value {dataInterface=&self:NiagaraDataInterfaceSkeletalMesh_0}
   }
  }
 ]
}
