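# Translucent, lit parent material for the syrup on the GroovyReaderGrid backpack, using a glass-like Fresnel over tiled noise normals projected from bounds-normalized, pre-skinned position.
# Fresnel drives opacity and a masked glow, with optional diffuse alpha masking, thin film and camera dither.
# The computed emissive becomes time-of-day-adjusted base color, leaving slight emission only at night, and opacity also varies by time of day.
# World position offset combines LOD, black-hole and optional vertex-hide effects.
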Material /BRCosmetics/Accessories/FORT_Backpacks/Backpack_F_MED_GroovyReaderGrid/Materials/M_MED_GroovyReaderGrid_Syrup_PARENT
graph {
 properties {
  bAutomaticallySetUsageInEditor=false
  bEnableResponsiveAA=true
  bScreenSpaceReflections=true
  bUseMaterialAttributes=true
  bUsedWithMeshDeformer=true
  bUsedWithMorphTargets=true
  bUsedWithSkeletalMesh=true
  blendMode=BLEND_Translucent
  ditherOpacityMask=true
  pixelDepthOffsetMode=PDOM_Legacy
  translucencyDirectionalLightingIntensity=5
  translucencyLightingMode=TLM_SurfacePerPixelLighting
  usedShadingModels=Default Lit
 }
 nodes [
  Constant3Vector() {desc=EmissiveColor, constant=<rgba 1 0.3325 0 1>}
  MaterialFunctionCall!(WorldPostionOfBlackHole, World Position) {materialFunction=/Game/Packages/Fortnite_UniqueMaterials_1/BlackHoleFunction}
  n1 = ScalarParameter() {parameterName=ThinFilmOpacity}
  n2 = TextureSampleParameter2D(UVs=TextureCoordinate()) {group=Mask Texture, parameterName=Diffuse Texture with Alpha Mask, sortPriority=10, texture=&T_BlacksRGB}
  n3 = TextureSampleParameter2D() {parameterName=Input S, texture=/Game/Global/Textures/Default/Blanks/T_BlackUserInterface}
  n4 = ScalarParameter() {defaultValue=1, parameterName=Fresnel Exponent}
  n5 = MaterialFunctionCall(Value_Morning=@n4, Value_Day=@n4, Value_Evening=@n4, Value_Night=Multiply(A=@n4) {constB=0.6}) {materialFunction=&Scalar_Time_of_Day}
  n6 = ObjectBounds()
  n7 = PreSkinnedLocalBounds()
  n8 = Saturate(None=Floor(None=ScalarParameter() {parameterName=Use Pre-Skinned Bounds, sliderMax=1}))
  n9 = ComponentMask(None=Divide(A=PreSkinnedPosition(), B=LinearInterpolate(A=@n6, B=@n7.Half Extents, Alpha=@n8))) {b=true, g=true}
  n10 = ScalarParameter() {defaultValue=0.2, parameterName=WaterTiling}
  n11 = TextureSampleParameter2D(UVs=VertexInterpolator(VS=Multiply(A=@n9, B=@n10)).PS) {
   parameterName=Noise Normal
   samplerSource=SSM_Wrap_WorldGroupSettings
   samplerType=SAMPLERTYPE_Normal
   texture=/Game/Global/Textures/Default/Noise/T_Noise_Normal_01
  }
  n12 = VectorParameter() {parameterName=Normal Intensity}
  n13 = TextureSampleParameter2D() {
   group=Mask Texture
   parameterName=Normals
   samplerSource=SSM_Wrap_WorldGroupSettings
   samplerType=SAMPLERTYPE_Normal
   texture=/Engine/EngineMaterials/FlatNormal
  }
  n14 = Normalize(VectorInput=Add(A=Multiply(A=@n11.RGB, B=@n12.RGB), B=@n13.RGB))
  n15 = Transform(None=@n14)
  n16 = Clamp(None=Multiply(A=Fresnel(ExponentIn=@n5.Result, Normal=@n15), B=ScalarParameter() {defaultValue=1, parameterName=Glass thickness}))
  n17 = LinearInterpolate(
   A=ScalarParameter() {parameterName=Fresnel Inner Emissive}
   B=ScalarParameter() {defaultValue=0.85, parameterName=Fresnel Outer Emissive}
   Alpha=@n16
  )
  n18 = VectorParameter() {parameterName=EmissiveColor}
  n19 = TextureSampleParameter2D() {parameterName=Input FX, texture=&T_BlackUserInterface}
  n20 = ScalarParameter() {parameterName=Add Custom Opacity, sortPriority=10}
  n21 = ScalarParameter() {parameterName=Fresnel Inner Transparency}
  n22 = ScalarParameter() {defaultValue=0.85, parameterName=Fresnel Outer Transparency}
  n23 = LinearInterpolate(
   A=Add(A=Multiply(A=@n19.B, B=@n20), B=LinearInterpolate(A=@n21, B=@n22, Alpha=@n16))
   B=@n19.B
   Alpha=ScalarParameter() {parameterName=Use Custom Opacity}
  )
  n24 = MakeMaterialAttributes!(
   BaseColor=StaticSwitchParameter(True=@n2.RGB, False=VectorParameter() {parameterName=Base Color}.RGB) {parameterName="Use Diffuse Texture for Color [ignores alpha channel]"}
   Metallic=LinearInterpolate(A=ScalarParameter() {defaultValue=0.1, parameterName=Metallic}, B=@n3.G, Alpha=ScalarParameter() {parameterName=Use Custom Metallic})
   Specular=LinearInterpolate(A=ScalarParameter() {defaultValue=1, parameterName=Specular}, B=@n3.R, Alpha=ScalarParameter() {parameterName=Use Custom Specular})
   Roughness=LinearInterpolate(A=ScalarParameter() {parameterName=Roughness}, B=@n3.B, Alpha=ScalarParameter() {parameterName=Use Custom Roughness})
   Anisotropy
   EmissiveColor=Multiply(
    A=Multiply(A=@n17, B=@n18.RGB)
    B=TextureSampleParameter2D() {parameterName=EmissiveMask, samplerType=SAMPLERTYPE_Masks, texture=&T_BlackMask}.RGB
   )
   Opacity=@n23
   OpacityMask=@n23
   Normal=@n14
   Tangent
   WorldPositionOffset
   SubsurfaceColor
   ClearCoat
   ClearCoatRoughness
   AmbientOcclusion
   Refraction
   CustomizedUVs_0
   CustomizedUVs_1
   CustomizedUVs_2
   CustomizedUVs_3
   CustomizedUVs_4
   CustomizedUVs_5
   CustomizedUVs_6
   CustomizedUVs_7
   PixelDepthOffset
   ShadingModel
   Displacement
  )
  n25 = GetMaterialAttributes(None=@n24) {attributeGetTypes [#1, #2]}
  n26 = Multiply(A=@n2.A, B=ScalarParameter() {defaultValue=1, group=Mask Texture, parameterName=Alpha Channel Mask Opacity})
  n27 = LinearInterpolate(A=@n25.Base Color, B=@n2.RGB, Alpha=@n26)
  n28 = Saturate(None=LinearInterpolate(A=@n25.Opacity, Alpha=@n26))
  n29 = StaticSwitchParameter(
   True=SetMaterialAttributes!(MaterialAttributes=@n25.MaterialAttributes, Base Color=Saturate(None=@n27), Opacity=@n28, Opacity Mask=@n28) {attributeSetTypes [#1, #2, #3]}
   False=@n24
  ) {group=Mask Texture, parameterName=Use Diffuse Texture with Alpha Mask, sortPriority=-50}
  n30 = Multiply(A=Subtract(A=ObjectPositionWS(), B=WorldPosition().XYZ), B=ScalarParameter() {parameterName=HT_CrunchVerts})
  n31 = VectorParameter() {parameterName=WorldPostionOfBlackHole, defaultValue=<rgba -30000 25536 1952 0>}
  n32 = MaterialFunctionCall!(
   World Position=WorldPosition().XYZ
   DistanceCullFadeAnimationOffset=MaterialFunctionCall() {materialFunction=/Game/Packages/Fortress_SharedMaterials/Base_Material_Functions/PickupEffectDistanceCullOverride}.Result
   Use actor center instead of pivot point
  ) {materialFunction=/Game/Packages/Fortress_SharedMaterials/Base_Material_Functions/LODEffect}
  n33 = WorldPosition()
  n34 = Add(
   A=MaterialFunctionCall(WorldPostionOfBlackHole=AppendVector(A=@n31.RGB, B=@n31.RGB), World Position=Add(A=@n32.Result, B=@n33.XYZ)) {materialFunction=&BlackHoleFunction}.Result
   B=@n32.Result
  )
  n35 = SetMaterialAttributes!(
   MaterialAttributes=StaticSwitchParameter(
    True=MaterialFunctionCall!(Alpha=@n1, Material Attributes=@n29) {materialFunction=/Game/Weapons/WeaponSkins/MaterialFunction/ThinFilmFunction}.Result
    False=@n29
   ) {parameterName=useThinFilm}
   World Position Offset=StaticSwitchParameter(True=Add(A=@n30, B=@n34), False=@n34) {parameterName=bUseVertexHide}
  ) {attributeSetTypes [#4]}
  n36 = MaterialFunctionCall!(In=@n35) {materialFunction=/Game/Packages/Fortress_SharedMaterials/Base_Material_Functions/BaseCustomizationFunctions/MF_FN_Vehicle_CameraDither}
  n37 = GetMaterialAttributes(None=MaterialFunctionCall(In=StaticSwitchParameter(True=@n36.Result, False=@n35) {parameterName=UseCameraDither}) {materialFunction=&MF_TechArtOverrides}.Result) {attributeGetTypes [#5, #2]}
  n38 = Multiply(A=@n37.Emissive Color) {constB=4.5}
  n39 = Constant3Vector() {constant=<rgba 0 -0.08 0.007 0>}
  n40 = Constant() {r=0.85}
  n41 = Constant() {r=0.81}
  n42 = Constant() {r=0.75}
  n43 = Constant3Vector()
  SetMaterialAttributes!(
   MaterialAttributes=@n37.MaterialAttributes
   Base Color=MaterialFunctionCall(
    Color_Morning=@n37.Emissive Color
    Color_Day=@n37.Emissive Color
    Color_Evening=Multiply(A=@n37.Emissive Color) {constB=2}
    Color_Night=Add(A=@n38, B=@n39)
   ) {materialFunction=&Color_Time_of_Day}.Result
   Opacity=Multiply(
    A=@n37.Opacity
    B=MaterialFunctionCall(Value_Morning=@n40, Value_Day=@n40, Value_Evening=@n41, Value_Night=@n42) {materialFunction=&Scalar_Time_of_Day}.Result
   )
   Emissive Color=MaterialFunctionCall(Color_Morning=@n43, Color_Day=@n43, Color_Evening=@n43, Color_Night=Multiply(A=@n37.Emissive Color) {constB=0.085}) {materialFunction=&Color_Time_of_Day}.Result
  ) {attributeSetTypes [#1, #2, #5]}
 ]
}
