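# Physics asset for the GroovyReaderGrid backpack cosmetic, providing secondary motion for its butter piece and drips.
# A central pack capsule connects to five small base anchor capsules through constraints with free swing and twist.
# Each anchor leads into a chain of one to three dyn capsules with per-joint swing limits and nearly no twist.
# The right drip and short mid-left drip swing much more widely than the other drips.
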PhysicsAsset /BRCosmetics/Accessories/FORT_Backpacks/Backpack_F_MED_GroovyReaderGrid/Meshes/F_MED_GroovyReaderGrid_Pack_Physics
bodies {
 base_butter [{shapeType=Capsule, length=0.505, radius=0.505, shapeName=base_butter_capsule, rotation=<rot 0 -90 0>}]
 base_drip_00_l [{shapeType=Capsule, length=0.505, radius=0.505, shapeName=base_drip_00_l_capsule, rotation=<rot 0 -90 0>}]
 base_drip_00_mid_r [{shapeType=Capsule, length=0.505, radius=0.505, shapeName=base_drip_00_mid_r_capsule, rotation=<rot 0 -90 0>}]
 base_drip_00_r [{shapeType=Capsule, length=0.505, radius=0.505, shapeName=base_drip_00_r_capsule, rotation=<rot 0 -90 0>}]
 base_drip_mid_l [{shapeType=Capsule, length=0.505, radius=0.505, shapeName=base_drip_mid_l_capsule, rotation=<rot 0 -90 0>}]
 dyn_butter [{shapeType=Capsule, length=2, radius=1, shapeName=dyn_butter_capsule, center=<2 0 0>, rotation=<rot 90 0 0>}]
 dyn_drip_00_l [{shapeType=Capsule, length=0, radius=0.805, shapeName=dyn_drip_00_l_capsule, center=<-0.379 0 0>, rotation=<rot 90 0 0>}]
 dyn_drip_00_mid_r [{shapeType=Capsule, length=0.0299, radius=1.1377, shapeName=dyn_drip_00_mid_r_capsule, center=<-0.0643 0 0>, rotation=<rot 90 0 0>}]
 dyn_drip_00_r [{shapeType=Capsule, length=0.0337, radius=0.9487, shapeName=dyn_drip_00_r_capsule, center=<-0.1485 0 0>, rotation=<rot 90 0 0>}]
 dyn_drip_01_l [{shapeType=Capsule, length=1.0856, radius=0.6924, shapeName=dyn_drip_01_l_capsule, center=<0.1584 0 0>, rotation=<rot 90 0 0>}]
 dyn_drip_01_mid_r [{shapeType=Capsule, length=0.0821, radius=1.0344, shapeName=dyn_drip_01_mid_r_capsule, center=<0.6131 0 0>, rotation=<rot 90 0 0>}]
 dyn_drip_01_r [{shapeType=Capsule, length=1.6291, radius=0.5984, shapeName=dyn_drip_01_r_capsule, center=<-0.2219 0 0>, rotation=<rot 90 0 0>}]
 dyn_drip_02_l [{shapeType=Capsule, length=0.1073, radius=0.8737, shapeName=dyn_drip_02_l_capsule, center=<0.9922 0 0>, rotation=<rot 90 0 0>}]
 dyn_drip_02_r [{shapeType=Capsule, length=0.5694, radius=0.7821, shapeName=dyn_drip_02_r_capsule, center=<0.7819 0 0>, rotation=<rot 90 0 0>}]
 dyn_drip_mid_l [{shapeType=Capsule, length=0.1761, radius=0.7182, shapeName=dyn_drip_mid_l_capsule, center=<0.163 0 0>, rotation=<rot 90 0 0>}]
 pack [{shapeType=Capsule, length=10.3674, radius=1.4956, shapeName=pack_capsule}]
}
constraints [
 {
  bone1Name=base_butter
  bone2Name=pack
  swing1LimitDegrees=45
  swing1Motion=Free
  swing2LimitDegrees=45
  swing2Motion=Free
  twistLimitDegrees=45
  twistMotion=Free
 }
 {bone1Name=dyn_butter, bone2Name=base_butter, swing1LimitDegrees=25, swing2LimitDegrees=10, twistLimitDegrees=1}
 {
  bone1Name=base_drip_00_l
  bone2Name=pack
  swing1LimitDegrees=45
  swing1Motion=Free
  swing2LimitDegrees=45
  swing2Motion=Free
  twistLimitDegrees=45
  twistMotion=Free
 }
 {bone1Name=dyn_drip_00_l, bone2Name=base_drip_00_l, swing1LimitDegrees=10, swing2LimitDegrees=10, twistLimitDegrees=1}
 {bone1Name=dyn_drip_01_l, bone2Name=dyn_drip_00_l, swing1LimitDegrees=9, swing2LimitDegrees=9, twistLimitDegrees=1}
 {bone1Name=dyn_drip_02_l, bone2Name=dyn_drip_01_l, swing1LimitDegrees=8.5, swing2LimitDegrees=8.5, twistLimitDegrees=1}
 {
  bone1Name=base_drip_00_mid_r
  bone2Name=pack
  swing1LimitDegrees=45
  swing1Motion=Free
  swing2LimitDegrees=45
  swing2Motion=Free
  twistLimitDegrees=45
  twistMotion=Free
 }
 {bone1Name=dyn_drip_00_mid_r, bone2Name=base_drip_00_mid_r, swing1LimitDegrees=8, swing2LimitDegrees=8, twistLimitDegrees=1}
 {bone1Name=dyn_drip_01_mid_r, bone2Name=dyn_drip_00_mid_r, swing1LimitDegrees=9, swing2LimitDegrees=9, twistLimitDegrees=1}
 {
  bone1Name=base_drip_00_r
  bone2Name=pack
  swing1LimitDegrees=45
  swing1Motion=Free
  swing2LimitDegrees=45
  swing2Motion=Free
  twistLimitDegrees=45
  twistMotion=Free
 }
 {bone1Name=dyn_drip_00_r, bone2Name=base_drip_00_r, swing1LimitDegrees=30, swing2LimitDegrees=25, twistLimitDegrees=1}
 {bone1Name=dyn_drip_01_r, bone2Name=dyn_drip_00_r, swing1LimitDegrees=42.5, swing2LimitDegrees=42.5, twistLimitDegrees=1}
 {bone1Name=dyn_drip_02_r, bone2Name=dyn_drip_01_r, swing1LimitDegrees=42.5, swing2LimitDegrees=42.5, twistLimitDegrees=1}
 {
  bone1Name=base_drip_mid_l
  bone2Name=pack
  swing1LimitDegrees=45
  swing1Motion=Free
  swing2LimitDegrees=45
  swing2Motion=Free
  twistLimitDegrees=45
  twistMotion=Free
 }
 {bone1Name=dyn_drip_mid_l, bone2Name=base_drip_mid_l, swing1LimitDegrees=50, swing2LimitDegrees=50, twistLimitDegrees=1}
]
